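# Material function that picks a brow flipbook frame Index (neutral, down, up, angry, worried or emote) from animation-driven curve parameters.
# Step nodes compare inner, outer, squeeze and emote pose values against tunable thresholds, with multiplied pairs acting as logical ANDs.
# A nested If chain resolves these tests in priority order to a constant index, using named reroutes for inputs and result.
# The directional pose curves read only R_ brow parameters.
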MaterialFunction /CosmeticCompanions/Materials/Flipbook/MaterialFunctions/MF_SK_Flipbook_Brows_Curves
graph {nodes [
  FunctionOutput(None=NamedRerouteUsage()) {outputName=Index}
  LinearInterpolate(
   B=NamedRerouteUsage()
   Alpha=Step(Y=NamedRerouteUsage(), X=ScalarParameter() {defaultValue=0.3, parameterName=Angry Threshold}) {constX=0.3, constY=0.3}
  )
  NamedRerouteDeclaration(None=ScalarParameter() {group="Flipbook 02.01 Brows | Curves (Driven By Animation)", parameterName=R_brow_Ot_Up_pose, sortPriority=3}) {name=Brow_Ot_Up_Pose, nodeColor=<rgba 1 0.2105 0 1>}
  NamedRerouteDeclaration(None=ScalarParameter() {defaultValue=0.1, group=Flipbook 02.02 Brows | Thresholds, parameterName=Threshold_Brow_Ot_Up, sortPriority=3}) {name=Threshold_Brow_Ot_Up, nodeColor=<rgba 1 0.0834 0.4399 1>}
  NamedRerouteDeclaration(None=Constant() {r=5}) {name=Index_Brow_Emote_01, nodeColor=<rgba 0.2818 0.3488 1 1>}
  NamedRerouteDeclaration(None=ScalarParameter() {group="Flipbook 02.01 Brows | Curves (Driven By Animation)", parameterName=R_brow_In_Dn_pose, sortPriority=0}) {name=Brow_In_Dn_Pose, nodeColor=<rgba 1 0.2105 0 1>}
  NamedRerouteDeclaration(None=ScalarParameter() {group="Flipbook 02.01 Brows | Curves (Driven By Animation)", parameterName=R_brow_Ot_Dn_pose, sortPriority=2}) {name=Brow_Ot_Dn_Pose, nodeColor=<rgba 1 0.2105 0 1>}
  NamedRerouteDeclaration(None=Constant()) {name=Index_Neutral, nodeColor=<rgba 0.2818 0.3488 1 1>}
  NamedRerouteDeclaration(None=Constant() {r=3}) {name=Index_Brows_Angry, nodeColor=<rgba 0.2818 0.3488 1 1>}
  NamedRerouteDeclaration(None=Constant() {r=2}) {name=Index_Brows_Up, nodeColor=<rgba 0.2818 0.3488 1 1>}
  NamedRerouteDeclaration(None=ScalarParameter() {group="Flipbook 02.01 Brows | Curves (Driven By Animation)", parameterName=brow_emote_01, sortPriority=3}) {name=Brow_Emote_01_Pose, nodeColor=<rgba 1 0.2105 0 1>}
  NamedRerouteDeclaration(None=Constant() {r=1}) {name=Index_Brows_Down, nodeColor=<rgba 0.2818 0.3488 1 1>}
  NamedRerouteDeclaration(None=ScalarParameter() {group="Flipbook 02.01 Brows | Curves (Driven By Animation)", parameterName=R_brow_In_Up_pose, sortPriority=1}) {name=Brow_In_Up_Pose, nodeColor=<rgba 1 0.2105 0 1>}
  NamedRerouteDeclaration(None=ScalarParameter() {defaultValue=0.1, group=Flipbook 02.02 Brows | Thresholds, parameterName=Threshold_Brow_In_Up, sortPriority=1}) {name=Threshold_Brow_In_Up, nodeColor=<rgba 1 0.0834 0.4399 1>}
  NamedRerouteDeclaration(None=ScalarParameter() {defaultValue=0.1, group=Flipbook 02.02 Brows | Thresholds, parameterName=Threshold_Brow_Emote_01, sortPriority=3}) {name=Threshold_Brow_Emote_01, nodeColor=<rgba 1 0.0834 0.4399 1>}
  NamedRerouteDeclaration(None=Constant() {r=4}) {name=Index_Brows_Worried, nodeColor=<rgba 0.2818 0.3488 1 1>}
  NamedRerouteDeclaration(None=ScalarParameter() {group="Flipbook 02.01 Brows | Curves (Driven By Animation)", parameterName=R_brow_squeeze_in_pose, sortPriority=3}) {name=R_Brow_Squeeze_In_Pose, nodeColor=<rgba 1 0.2105 0 1>}
  NamedRerouteDeclaration(None=ScalarParameter() {defaultValue=0.5, group=Flipbook 02.02 Brows | Thresholds, parameterName=Threshold_Brow_Squeeze_In, sortPriority=3}) {name=Threshold_Brow_Squeeze_In, nodeColor=<rgba 1 0.0834 0.4399 1>}
  n1 = Step(Y=NamedRerouteUsage(), X=NamedRerouteUsage())
  n2 = Step(Y=NamedRerouteUsage(), X=NamedRerouteUsage())
  n3 = NamedRerouteUsage()
  n4 = Multiply(A=Step(Y=NamedRerouteUsage(), X=NamedRerouteUsage()), B=Step(Y=NamedRerouteUsage(), X=NamedRerouteUsage()))
  n5 = NamedRerouteUsage()
  n6 = NamedRerouteUsage()
  n7 = NamedRerouteUsage()
  n8 = NamedRerouteUsage()
  n9 = NamedRerouteUsage()
  n10 = NamedRerouteUsage()
  n11 = Step(Y=NamedRerouteUsage(), X=NamedRerouteUsage())
  n12 = Step(Y=NamedRerouteUsage(), X=NamedRerouteUsage())
  n13 = NamedRerouteUsage()
  n14 = NamedRerouteUsage()
  n15 = NamedRerouteUsage()
  n16 = NamedRerouteUsage()
  n17 = NamedRerouteUsage()
  n18 = If(
   A=Multiply(A=Step(Y=@n6, X=@n7), B=Step(Y=@n8, X=@n9))
   "A > B"=@n10
   "A == B"=@n10
   "A < B"=If(A=Multiply(A=@n11, B=@n12), "A > B"=@n13, "A == B"=@n13, "A < B"=If(A=@n14, B=@n15, "A > B"=@n16, "A < B"=@n17)) {constB=1}
  ) {constB=1}
  NamedRerouteDeclaration(None=If(A=Multiply(A=@n1, B=@n2), "A > B"=@n3, "A == B"=@n3, "A < B"=If(A=@n4, "A > B"=@n5, "A == B"=@n5, "A < B"=@n18) {constB=1}) {constB=1}) {name=Final Brow Output, nodeColor=<rgba 1 0.0234 0 1>}
  NamedRerouteDeclaration(None=ScalarParameter() {defaultValue=0.1, group=Flipbook 02.02 Brows | Thresholds, parameterName=Threshold_Brow_In_Dn, sortPriority=0}) {name=Threshold_Brow_In_Dn, nodeColor=<rgba 1 0.0834 0.4399 1>}
  NamedRerouteDeclaration(None=ScalarParameter() {defaultValue=0.1, group=Flipbook 02.02 Brows | Thresholds, parameterName=Threshold_Brow_Ot_Dn, sortPriority=2}) {name=Threshold_Brow_Ot_Dn, nodeColor=<rgba 1 0.0834 0.4399 1>}
  NamedRerouteUsage()
 ]}
