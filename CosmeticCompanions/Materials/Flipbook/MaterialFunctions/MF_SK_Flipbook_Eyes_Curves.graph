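# Material function for cosmetic companions that turns animation-driven eye pose curves into an Index for an eye flipbook.
# Scalar parameters supply right-side blink, wide, lid and smile poses plus an eye emote, alongside tunable thresholds and constant expression indices.
# Nested If comparisons choose a base index, then Step and Max masks drive chained lerps that override it.
# Named reroutes carry the inputs, indices and final eye result.
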MaterialFunction /CosmeticCompanions/Materials/Flipbook/MaterialFunctions/MF_SK_Flipbook_Eyes_Curves
graph {nodes [
  FunctionOutput(None=NamedRerouteUsage()) {outputName=Index}
  If(A=NamedRerouteUsage(), B=NamedRerouteUsage(), "A > B"=NamedRerouteUsage())
  NamedRerouteDeclaration(None=Constant() {r=1}) {name=Index_Closed_Happy, nodeColor=<rgba 0.2818 0.3488 1 1>}
  NamedRerouteDeclaration(None=ScalarParameter() {parameterName=R_smile_pose}) {name=Smile Pose, nodeColor=<rgba 1 0.2105 0 1>}
  NamedRerouteDeclaration(None=Constant() {r=3}) {name=Index_Eyes_Squint, nodeColor=<rgba 0.2818 0.3488 1 1>}
  NamedRerouteDeclaration(None=Constant() {r=2}) {name=Index_Closed, nodeColor=<rgba 0.2818 0.3488 1 1>}
  NamedRerouteDeclaration(None=Constant()) {name=Index_Upper_Lid_Down, nodeColor=<rgba 0.2818 0.3488 1 1>}
  NamedRerouteDeclaration(None=ScalarParameter() {defaultValue=0.4, group=Flipbook 03.02 Eyes | Thresholds, parameterName=Threshold_Blink, sortPriority=0}) {name=Threshold_Blink, nodeColor=<rgba 1 0.0834 0.4399 1>}
  NamedRerouteDeclaration(None=ScalarParameter() {defaultValue=0.3, group=Flipbook 03.02 Eyes | Thresholds, parameterName=Threshold_Lower_Lid_Up, sortPriority=2}) {name=Threshold_Lower_Lid_Up, nodeColor=<rgba 1 0.0834 0.4399 1>}
  NamedRerouteDeclaration(None=ScalarParameter() {group="Flipbook 03.01 Eyes | Curves (Driven By Animation)", parameterName=R_upperLidDown_pose, sortPriority=3}) {name=Upper_Lid_Down_Pose, nodeColor=<rgba 1 0.2105 0 1>}
  NamedRerouteDeclaration(None=ScalarParameter() {defaultValue=0.4, group=Flipbook 03.02 Eyes | Thresholds, parameterName=Threshold_Upper_Lid_Down, sortPriority=3}) {name=Threshold_Upper_Lid_Down, nodeColor=<rgba 1 0.0834 0.4399 1>}
  NamedRerouteDeclaration(None=ScalarParameter() {group="Flipbook 03.01 Eyes | Curves (Driven By Animation)", parameterName=R_wide_pose, sortPriority=1}) {name=Wide_Pose, nodeColor=<rgba 1 0.2105 0 1>}
  NamedRerouteDeclaration(None=ScalarParameter() {defaultValue=0.4, group=Flipbook 03.02 Eyes | Thresholds, parameterName=Threshold_Wide, sortPriority=1}) {name=Threshold_Wide, nodeColor=<rgba 1 0.0834 0.4399 1>}
  NamedRerouteDeclaration(None=Constant()) {name=Index_Wide, nodeColor=<rgba 0.2818 0.3488 1 1>}
  NamedRerouteDeclaration(None=ScalarParameter() {defaultValue=0.4, parameterName=Threshold_Smile}) {name=Threshold_Smile, nodeColor=<rgba 1 0.0834 0.4399 1>}
  NamedRerouteDeclaration(None=ScalarParameter() {group="Flipbook 03.01 Eyes | Curves (Driven By Animation)", parameterName=eye_emote_01}) {name=Eye_Emote_01_Pose, nodeColor=<rgba 1 0.2105 0 1>}
  NamedRerouteDeclaration(None=ScalarParameter() {defaultValue=0.4, group=Flipbook 03.02 Eyes | Thresholds, parameterName=Threshold_Eye_Emote_01, sortPriority=3}) {name=Threshold_Eye_Emote_01, nodeColor=<rgba 1 0.0834 0.4399 1>}
  n1 = If(
   A=NamedRerouteUsage()
   B=NamedRerouteUsage()
   "A > B"=NamedRerouteUsage()
   "A < B"=If(A=NamedRerouteUsage(), B=NamedRerouteUsage(), "A > B"=NamedRerouteUsage(), "A < B"=NamedRerouteUsage())
  )
  n2 = NamedRerouteUsage()
  n3 = Max(A=Step(Y=NamedRerouteUsage(), X=NamedRerouteUsage()), B=Step(Y=NamedRerouteUsage(), X=NamedRerouteUsage()))
  n4 = Step(Y=NamedRerouteUsage(), X=NamedRerouteUsage())
  NamedRerouteDeclaration(None=LinearInterpolate(A=LinearInterpolate(A=@n1, B=@n2, Alpha=@n3), B=NamedRerouteUsage(), Alpha=Multiply(A=@n4, B=@n3))) {name=Final Eye Output, nodeColor=<rgba 1 0.0234 0 1>}
  NamedRerouteDeclaration(None=Constant() {r=4}) {name=Index_Eye_Emote_01, nodeColor=<rgba 0.2818 0.3488 1 1>}
  NamedRerouteDeclaration(None=ScalarParameter() {group="Flipbook 03.01 Eyes | Curves (Driven By Animation)", parameterName=R_blink_pose, sortPriority=0}) {name=Blink_Pose, nodeColor=<rgba 1 0.2105 0 1>}
  NamedRerouteDeclaration(None=ScalarParameter() {group="Flipbook 03.01 Eyes | Curves (Driven By Animation)", parameterName=R_lower_lid_up_pose, sortPriority=2}) {name=Lower_Lid_Up_Pose, nodeColor=<rgba 1 0.2105 0 1>}
  NamedRerouteDeclaration(None=Constant()) {name=Index_Neutral, nodeColor=<rgba 0.2818 0.3488 1 1>}
 ]}
