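# Base anime-shaded material for skeletal cosmetic companions, with animated flipbook faces on the Cloth shading model.
# Curve functions pick remappable brow, eye and mouth indices, driving texture flipbook sampling or vertex-color masks that hide geometry through WPO.
# An optional eye animation function overrides flipbook eyes.
# Anime shader Pre, Calculations (with a sphere-normal override) and Post stages produce the surface, with toggles for face color, emissive, opacity, WPO and debug views.
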Material /CosmeticCompanions/Materials/Flipbook/M_SK_Anime_Flipbook_Base_V1
graph {
 properties {
  bAutomaticallySetUsageInEditor=false
  bUseMaterialAttributes=true
  bUsedWithClothing=true
  bUsedWithMeshDeformer=true
  bUsedWithMorphTargets=true
  bUsedWithNiagaraMeshParticles=true
  bUsedWithSkeletalMesh=true
  bUsedWithStaticLighting=true
  pixelDepthOffsetMode=PDOM_Legacy
  shadingModel=MSM_Cloth
  usedShadingModels=Cloth
 }
 nodes [
  NamedRerouteDeclaration(None=StaticBoolParameter() {group=Flipbook 06 Debug, parameterName=Flipbook.UseDebugVertColor, sortPriority=-1}) {name=UseDebugVertColor, nodeColor=<rgba 0.1953 1 0 1>}
  NamedRerouteDeclaration(None=ChannelMaskParameter(None=VertexColor()) {group=Flipbook 02 Brows, parameterName=Flipbook.BrowsVertexColor}.R) {name=Brows_VertexColor, nodeColor=<rgba 1 0.9844 0 1>}
  n1 = ScalarParameter() {
   controlType=Enumeration
   enumeration=/Game/Global/Materials/Enums/E_Mat_OnOff
   group=Flipbook 01 Global
   parameterName=Flipbook.ApplyOpacityMask
   sortPriority=4
  }
  NamedRerouteDeclaration(None=LinearInterpolate(A=Constant(), B=NamedRerouteUsage(), Alpha=@n1)) {name=Flipbook_Opacity, nodeColor=<rgba 0 1 0.6016 1>}
  n2 = NamedRerouteUsage()
  n3 = Multiply(A=NamedRerouteUsage(), B=ScalarParameter() {defaultValue=1, group=Flipbook 01 Global, parameterName=Flipbook.EmissiveStrength})
  n4 = NamedRerouteUsage()
  n5 = ScalarParameter() {
   controlType=Enumeration
   enumeration=&E_Mat_OnOff
   group=Flipbook 01 Global
   parameterName=Flipbook.ApplyEmissive
   sortPriority=3
  }
  NamedRerouteDeclaration(None=LinearInterpolate(A=NamedRerouteUsage(), B=LinearInterpolate(A=@n2, B=@n3, Alpha=@n4), Alpha=@n5)) {name=Flipbook_Emissive, nodeColor=<rgba 0.108 0.2905 1 1>}
  NamedRerouteDeclaration(None=StaticBoolParameter() {group=Flipbook 02 Brows, parameterName=Flipbook.UseTextureForBrows, sortPriority=-20}) {name=UseTextureForBrows, nodeColor=<rgba 0 1 0.3438 1>}
  n6 = TextureSampleParameter2D() {
   group=Eye Animation 01 Global
   parameterName=EyeAnimation.EyeMask
   samplerSource=SSM_Wrap_WorldGroupSettings
   samplerType=SAMPLERTYPE_Masks
   sortPriority=3
   texture=&T_WhiteMask
  }
  n7 = VertexColor()
  n8 = ScalarParameter() {
   controlType=Enumeration
   enumeration=&E_Mat_OnOff
   group=Eye Animation 01 Global
   parameterName=EyeAnimation.UseVertexColorForEyeMask
   sortPriority=0
  }
  n9 = LinearInterpolate(A=@n6.RGBA, B=AppendVector(A=@n7, B=@n7), Alpha=@n8)
  n10 = TextureCoordinate() {coordinateIndex=1}
  n11 = Sign(None=Subtract(A=ComponentMask(None=@n10) {r=true}) {constB=0.0001})
  n12 = MaterialFunctionCall!(
   Color Lit=NamedRerouteUsage()
   Color Shaded=NamedRerouteUsage()
   Use UV2=StaticBoolParameter() {defaultValue=true, group=Eye Animation 01 Global, parameterName=EyeAnimation.UseUV2, sortPriority=-10}
   Mask=StaticSwitchParameter(
    True=ChannelMaskParameter(None=@n9) {group=Eye Animation 01 Global, parameterName=EyeAnimation.EyeMaskChannel, sortPriority=1}.R
    False=Saturate(None=@n11)
   ) {group=Eye Animation 01 Global, parameterName=EyeAnimation.UseEyeMask, sortPriority=0}
  ) {materialFunction=/CosmeticCompanions/Materials/MaterialFunctions/MF_SK_EyeAnimation}
  NamedRerouteDeclaration(None=StaticSwitch(True=@n12.Shaded, False=NamedRerouteUsage(), Value=NamedRerouteUsage())) {name=EeyAnimColor_Shaded, nodeColor=<rgba 0.125 1 0 1>}
  n13 = MaterialFunctionCall!(Material=NamedRerouteUsage(), Vertex Color Hide Mask=NamedRerouteUsage()) {materialFunction=/Game/Global/Materials/Character_Parent_2023/FlipbookFace_Parent_2023/MaterialFunctions/MF_Flipbook_Debug}
  NamedRerouteDeclaration(None=@n13.Debug Vert Color) {name=DebugVertColor, nodeColor=<rgba 0 0.6953 1 1>}
  NamedRerouteDeclaration(None=@n13.Debug Vert Color WPO) {name=DebugVertColor_WPO, nodeColor=<rgba 1 0.9844 0 1>}
  n14 = Constant() {r=1}
  n15 = NamedRerouteUsage()
  n16 = ScalarParameter() {defaultValue=0.1, group=Flipbook 01 Global, parameterName=Flipbook.ExtraVertexColorValue, sortPriority=-6}
  NamedRerouteDeclaration(None=LinearInterpolate(
    A=MaterialFunctionCall!(Hide Element Toggle=@n14, VertexColorChannel=@n15, Vertex Value to hide=@n16) {materialFunction=/Game/Packages/Fortress_SharedMaterials/Base_Material_Functions/VertexColorMask}.Result
    B=Constant()
    Alpha=ScalarParameter() {desc=Set by CID, group=Flipbook 01 Global, parameterName=Flipbook.HideExtraVertexColor, sortPriority=-5}
   )) {name=ExtraVertexMask, nodeColor=<rgba 0 1 0.3906 1>}
  NamedRerouteDeclaration(None=MaterialFunctionCall!(
    Index=MaterialFunctionCall() {materialFunction=/CosmeticCompanions/Materials/Flipbook/MaterialFunctions/MF_SK_Flipbook_Brows_Curves}.Index
    Input_0-3=NamedRerouteUsage()
    Input_4-7=NamedRerouteUsage()
   ) {materialFunction=/Game/Global/Materials/Character_Parent_2023/FlipbookFace_Parent_2023/MaterialFunctions/MF_RemapIndex_8}.Remapped Index) {name=Brows_CurveIndex, nodeColor=<rgba 1 0.4453 0 1>}
  NamedRerouteDeclaration(None=@n13.Debug UV WPO) {name=DebugUV_WPO, nodeColor=<rgba 1 0 0.3047 1>}
  NamedRerouteDeclaration(None=StaticBoolParameter() {group=Flipbook 01 Global, parameterName=Flipbook.UseBrows, sortPriority=-10}) {name=UseBrows, nodeColor=<rgba 0.1953 1 0 1>}
  n17 = MaterialFunctionCall!(
   Curve Index=NamedRerouteUsage()
   Use Texture=NamedRerouteUsage()
   ----------------------
   Vertex Color=NamedRerouteUsage()
   ---------------------
   Flipbook Texture=NamedRerouteUsage()
   Column Count=NamedRerouteUsage()
   Row Count=NamedRerouteUsage()
   UVs=NamedRerouteUsage()
   UV Offset=NamedRerouteUsage()
   UV Scale=NamedRerouteUsage()
  ) {materialFunction=/Game/Global/Materials/Character_Parent_2023/FlipbookFace_Parent_2023/MaterialFunctions/MF_Sample_Flipbook}
  NamedRerouteDeclaration(None=@n17.Flipbook Vertex Mask) {name=BrowsVertexMask, nodeColor=<rgba 1 0 0.8438 1>}
  NamedRerouteDeclaration(None=StaticBoolParameter() {group=Eye Animation 01 Global, parameterName="UseEyeAnimation (Overrides Flipbook Eyes)", sortPriority=-30}) {name=UseEyeAnimation, nodeColor=<rgba 0.1953 1 0 1>}
  NamedRerouteDeclaration(None=@n17.Flipbook Alpha) {name=BrowsAlpha, nodeColor=<rgba 1 0 0.8438 1>}
  n18 = VectorParameter() {
   group=Flipbook 04 Mouth
   parameterName=Flipbook.MouthRemap_4-7
   sortPriority=202
   channelNames {a=7 - Mouth Frown Open, b=6 - Mouth Frown Closed, g=5 - Mouth Smile Open, r=4 - Mouth Smile Closed}
   defaultValue=<rgba 4 5 6 7>
  }
  NamedRerouteDeclaration(None=@n18.RGBA) {name=Mouth_Remap_4-7, nodeColor=<rgba 1 0.6652 0.0014 1>}
  n19 = VectorParameter() {
   group=Flipbook 04 Mouth
   parameterName=Flipbook.MouthRemap_8-11
   sortPriority=203
   channelNames {a=11 - Oo Phoneme, b=10 - Ee Phoneme, g=9 - Mouth Grimmace Open, r=8 - Mouth Grimmace Closed}
   defaultValue=<rgba 8 9 10 11>
  }
  NamedRerouteDeclaration(None=@n19.RGBA) {name=Mouth_Remap_8-11, nodeColor=<rgba 1 0.6652 0.0014 1>}
  n20 = NamedRerouteUsage()
  n21 = Multiply(A=NamedRerouteUsage(), B=ScalarParameter() {defaultValue=1, parameterName=FlipbookFaceAnimeShadedValue})
  n22 = NamedRerouteUsage()
  n23 = ScalarParameter() {
   controlType=Enumeration
   defaultValue=1
   enumeration=&E_Mat_OnOff
   group=Flipbook 01 Global
   parameterName=Flipbook.ApplyBaseColor
   sortPriority=2
  }
  NamedRerouteDeclaration(None=LinearInterpolate(A=NamedRerouteUsage(), B=LinearInterpolate(A=@n20, B=@n21, Alpha=@n22), Alpha=@n23)) {name=Flipbook_ColorShaded, nodeColor=<rgba 1 0 0.0937 1>}
  n24 = NamedRerouteUsage()
  n25 = NamedRerouteUsage()
  n26 = NamedRerouteUsage()
  n27 = ScalarParameter() {
   controlType=Enumeration
   defaultValue=1
   enumeration=&E_Mat_OnOff
   group=Flipbook 01 Global
   parameterName=Flipbook.ApplyBaseColor
   sortPriority=2
  }
  NamedRerouteDeclaration(None=LinearInterpolate(A=NamedRerouteUsage(), B=LinearInterpolate(A=@n24, B=@n25, Alpha=@n26), Alpha=@n27)) {name=Flipbook_ColorLit, nodeColor=<rgba 0 0.4141 1 1>}
  NamedRerouteDeclaration(None=StaticBoolParameter() {group=Flipbook 06 Debug, parameterName=Flipbook.UseDebugUVs, sortPriority=-1}) {name=UseDebugUVs, nodeColor=<rgba 0.1953 1 0 1>}
  n28 = VectorParameter() {
   group=Flipbook 04 Mouth
   parameterName=Flipbook.MouthRemap_12-15
   sortPriority=204
   channelNames {a=15 - Unused, b=14 - Unused, g=13 - Unused, r=12 - Custom Mouth}
   defaultValue=<rgba 12 13 14 15>
  }
  NamedRerouteDeclaration(None=@n28.RGBA) {name=Mouth_Remap_12-15, nodeColor=<rgba 1 0.6652 0.0014 1>}
  NamedRerouteDeclaration(None=MaterialFunctionCall!(
    Index=MaterialFunctionCall() {materialFunction=/CosmeticCompanions/Materials/Flipbook/MaterialFunctions/MF_SK_Flipbook_Mouth_Curves}.Index
    Input_0-3=NamedRerouteUsage()
    Input_4-7=NamedRerouteUsage()
    Input_8-11=NamedRerouteUsage()
    Input_12-15=NamedRerouteUsage()
   ) {materialFunction=/Game/Global/Materials/Character_Parent_2023/FlipbookFace_Parent_2023/MaterialFunctions/MF_RemapIndex_16}.Remapped Index) {name=Mouth_CurveIndex, nodeColor=<rgba 1 0.4453 0 1>}
  n29 = VectorParameter() {
   group=Flipbook 03 Eyes
   parameterName=Flipbook.EyesRemap_0-3
   sortPriority=201
   channelNames {a=3 - Eyes Wide CC, b=2 - Eyes Squint CC, g=1 - Eyes Closed, r=0 - Eyes Neutral CC}
   defaultValue=<rgba 0 1 2 3>
  }
  NamedRerouteDeclaration(None=@n29.RGBA) {name=Eyes_Remap_0-3, nodeColor=<rgba 1 0.6652 0.0014 1>}
  n30 = VectorParameter() {
   group=Flipbook 03 Eyes
   parameterName=Flipbook.EyesRemap_4-7
   sortPriority=202
   channelNames {a=7 - Unused, b=6 - Unused, g=5 - Unused, r=4 - Custom Eyes}
   defaultValue=<rgba 4 5 6 7>
  }
  NamedRerouteDeclaration(None=@n30.RGBA) {name=Eyes_Remap_4-7, nodeColor=<rgba 1 0.6652 0.0014 1>}
  n31 = MaterialFunctionCall!(CustomUV) {materialFunction=/Game/Global/Materials/Character_Parent_2023/Anime_Parent_2023/MaterialFunctions/MFC_AnimeShader_2023_Pre}
  NamedRerouteDeclaration(None=@n31.Color_Lit) {name=Color_Lit, nodeColor=<rgba 0 1 0.8359 1>}
  n32 = MaterialFunctionCall(
   Curve Index=NamedRerouteUsage()
   Use Texture=NamedRerouteUsage()
   Vertex Color=NamedRerouteUsage()
   Flipbook Texture=NamedRerouteUsage()
   Column Count=NamedRerouteUsage()
   Row Count=NamedRerouteUsage()
   UVs=NamedRerouteUsage()
   UV Offset=NamedRerouteUsage()
   UV Scale=NamedRerouteUsage()
  ) {materialFunction=&MF_Sample_Flipbook}
  NamedRerouteDeclaration(None=@n32.Flipbook Vertex Mask) {name=EyesVertexMask, nodeColor=<rgba 0.2436 0 1 1>}
  n33 = TextureObjectParameter() {
   group=Flipbook 02 Brows
   parameterName=Flipbook.BrowsTexture
   sortPriority=2
   texture=/Game/Global/Materials/Character_Parent_2023/FlipbookFace_Parent_2023/Textures/T_black_emptyalpha
  }
  NamedRerouteDeclaration(None=@n33) {name=Brows_FlipbookTexture, nodeColor=<rgba 1 0.1875 0 1>}
  NamedRerouteDeclaration(None=@n32.Flipbook Color) {name=EyesColor, nodeColor=<rgba 0.2436 0 1 1>}
  NamedRerouteDeclaration(None=ScalarParameter() {defaultValue=3, group=Flipbook 02 Brows, parameterName=Flipbook.BrowsRowCount, sortPriority=3}) {name=Brows_RowCount, nodeColor=<rgba 1 0 0.7734 1>}
  NamedRerouteDeclaration(None=ScalarParameter() {defaultValue=3, group=Flipbook 02 Brows, parameterName=Flipbook.BrowsColumnCount, sortPriority=4}) {name=Brows_ColumnCount, nodeColor=<rgba 1 0 0.7734 1>}
  NamedRerouteDeclaration(None=StaticSwitchParameter(True=TextureCoordinate() {coordinateIndex=1}, False=TextureCoordinate()) {group=Flipbook 02 Brows, parameterName=Flipbook.BrowsUseUV2, sortPriority=10}) {name=Brows_UVs, nodeColor=<rgba 1 0.5859 0 1>}
  n34 = ScalarParameter() {defaultValue=1, group=Flipbook 02 Brows, parameterName=Flipbook.BrowsUVScaleX, sortPriority=12}
  n35 = ScalarParameter() {defaultValue=1, group=Flipbook 02 Brows, parameterName=Flipbook.BrowsUVScaleY, sortPriority=13}
  NamedRerouteDeclaration(None=Multiply(
    A=ScalarParameter() {defaultValue=1, group=Flipbook 02 Brows, parameterName=Flipbook.BrowsGlobalScale, sortPriority=12}
    B=AppendVector(A=@n34, B=@n35)
   )) {name=Brows_UVScale, nodeColor=<rgba 0.2656 0 1 1>}
  NamedRerouteDeclaration(None=AppendVector(
    A=ScalarParameter() {group=Flipbook 02 Brows, parameterName=Flipbook.BrowsUVOffsetX, sortPriority=20}
    B=ScalarParameter() {group=Flipbook 02 Brows, parameterName=Flipbook.BrowsUVOffsetY, sortPriority=21}
   )) {name=Brows_UVOffset, nodeColor=<rgba 0.0547 1 0 1>}
  NamedRerouteDeclaration(None=@n32.Flipbook Alpha) {name=EyesAlpha, nodeColor=<rgba 0.2436 0 1 1>}
  n36 = StaticBoolParameter() {
   defaultValue=true
   desc="Curves are shared with the material Anime Eye Animation functionality(found in \"01_EyeAnimationPoses\"). When the material Eye Animation is enabled, flipbook eyes will be disabled."
   group=Flipbook 01 Global
   parameterName=Flipbook.UseEyes
   sortPriority=-9
  }
  NamedRerouteDeclaration(None=@n36) {name=UseEyes, nodeColor=<rgba 0.1953 1 0 1>}
  n37 = StaticBoolParameter() {
   desc="Needed for when additional areas of the mesh use vertex colors for hiding geo. Avoiding conflict with the filpbooks vertex mask functionality. "
   group=Flipbook 01 Global
   parameterName=Flipbook.UseExtraVertexMask
   sortPriority=-7
  }
  NamedRerouteDeclaration(None=@n37) {name=UseExtraVertexMask, nodeColor=<rgba 0.1953 1 0 1>}
  NamedRerouteDeclaration(None=StaticBoolParameter() {defaultValue=true, group=Flipbook 01 Global, parameterName=Flipbook.UseMouth, sortPriority=-8}) {name=UseMouth, nodeColor=<rgba 0.1953 1 0 1>}
  n38 = VectorParameter() {
   group=Flipbook 02 Brows
   parameterName=Flipbook.BrowsRemap_0-3
   sortPriority=201
   channelNames {a=3 - Brows Worried, b=2 - Brows Angry, g=1 - Brows Up, r=0 - Brows Neutral}
   defaultValue=<rgba 0 1 2 3>
  }
  NamedRerouteDeclaration(None=@n38.RGBA) {name=Brows_Remap_0-3, nodeColor=<rgba 1 0.6652 0.0014 1>}
  NamedRerouteDeclaration(None=MaterialFunctionCall(
    Index=MaterialFunctionCall() {materialFunction=/CosmeticCompanions/Materials/Flipbook/MaterialFunctions/MF_SK_Flipbook_Eyes_Curves}.Index
    Input_0-3=NamedRerouteUsage()
    Input_4-7=NamedRerouteUsage()
   ) {materialFunction=&MF_RemapIndex_8}.Remapped Index) {name=Eyes_CurveIndex, nodeColor=<rgba 1 0.4453 0 1>}
  NamedRerouteDeclaration(None=StaticBoolParameter() {group=Flipbook 03 Eyes, parameterName=Flipbook.UseTextureForEyes, sortPriority=-20}) {name=UseTextureForEyes, nodeColor=<rgba 0 1 0.3438 1>}
  NamedRerouteDeclaration(None=ScalarParameter() {defaultValue=3, group=Flipbook 03 Eyes, parameterName=Flipbook.EyesRowCount, sortPriority=3}) {name=Eyes_RowCount, nodeColor=<rgba 1 0 0.7734 1>}
  NamedRerouteDeclaration(None=ScalarParameter() {defaultValue=3, group=Flipbook 03 Eyes, parameterName=Flipbook.EyesColumnCount, sortPriority=4}) {name=Eyes_ColumnCount, nodeColor=<rgba 1 0 0.7734 1>}
  NamedRerouteDeclaration(None=@n31.Color_Shaded) {name=Color_Shaded, nodeColor=<rgba 1 0.5156 0 1>}
  NamedRerouteDeclaration(None=StaticSwitchParameter(True=TextureCoordinate() {coordinateIndex=1}, False=TextureCoordinate()) {group=Flipbook 03 Eyes, parameterName=Flipbook.EyesUseUV2, sortPriority=10}) {name=Eyes_UVs, nodeColor=<rgba 1 0.5859 0 1>}
  n39 = ScalarParameter() {defaultValue=1, group=Flipbook 03 Eyes, parameterName=Flipbook.EyesUVScaleX, sortPriority=12}
  n40 = ScalarParameter() {defaultValue=1, group=Flipbook 03 Eyes, parameterName=Flipbook.EyesUVScaleY, sortPriority=13}
  NamedRerouteDeclaration(None=Multiply(
    A=ScalarParameter() {defaultValue=1, group=Flipbook 03 Eyes, parameterName=Flipbook.EyesGlobalScale, sortPriority=12}
    B=AppendVector(A=@n39, B=@n40)
   )) {name=Eyes_UVScale, nodeColor=<rgba 0.2656 0 1 1>}
  NamedRerouteDeclaration(None=AppendVector(
    A=ScalarParameter() {group=Flipbook 03 Eyes, parameterName=Flipbook.EyesUVOffsetX, sortPriority=20}
    B=ScalarParameter() {group=Flipbook 03 Eyes, parameterName=Flipbook.EyesUVOffsetY, sortPriority=21}
   )) {name=Eyes_UVOffset, nodeColor=<rgba 0.0547 1 0 1>}
  NamedRerouteDeclaration(None=ChannelMaskParameter(None=VertexColor()) {group=Flipbook 03 Eyes, parameterName=Flipbook.EyesVertexColor}.R) {name=Eyes_VertexColor, nodeColor=<rgba 1 0.9844 0 1>}
  n41 = TextureObjectParameter() {
   group=Flipbook 03 Eyes
   parameterName=Flipbook.EyesTexture
   sortPriority=2
   texture=&T_black_emptyalpha
  }
  NamedRerouteDeclaration(None=@n41) {name=Eyes_FlipbookTexture, nodeColor=<rgba 1 0.1875 0 1>}
  n42 = MaterialFunctionCall(
   Curve Index=NamedRerouteUsage()
   Use Texture=NamedRerouteUsage()
   Vertex Color=NamedRerouteUsage()
   Flipbook Texture=NamedRerouteUsage()
   Column Count=NamedRerouteUsage()
   Row Count=NamedRerouteUsage()
   UVs=NamedRerouteUsage()
   UV Offset=NamedRerouteUsage()
   UV Scale=NamedRerouteUsage()
  ) {materialFunction=&MF_Sample_Flipbook}
  NamedRerouteDeclaration(None=@n42.Flipbook Vertex Mask) {name=MouthVertexMask, nodeColor=<rgba 0.2092 0.8266 1 1>}
  NamedRerouteDeclaration(None=@n42.Flipbook Color) {name=MouthColor, nodeColor=<rgba 0.2092 0.8266 1 1>}
  NamedRerouteDeclaration(None=@n42.Flipbook Alpha) {name=MouthAlpha, nodeColor=<rgba 0.2092 0.8266 1 1>}
  NamedRerouteDeclaration(None=@n17.Flipbook Color) {name=BrowsColor, nodeColor=<rgba 1 0 0.8438 1>}
  n43 = VectorParameter() {
   group=Flipbook 02 Brows
   parameterName=Flipbook.BrowsRemap_4-7
   sortPriority=202
   channelNames {a=7 - Unused, b=6 - Unused, g=5 - Unused, r=4 - Custom Brows}
   defaultValue=<rgba 4 5 6 7>
  }
  NamedRerouteDeclaration(None=@n43.RGBA) {name=Brows_Remap_4-7, nodeColor=<rgba 1 0.6652 0.0014 1>}
  NamedRerouteDeclaration(None=StaticSwitch(True=@n12.Lit, False=NamedRerouteUsage(), Value=NamedRerouteUsage())) {name=EyeAnimColor_Lit, nodeColor=<rgba 0.1953 0 1 1>}
  n44 = NamedRerouteUsage()
  n45 = Constant()
  n46 = NamedRerouteUsage()
  n47 = Add(
   A=StaticSwitch(True=NamedRerouteUsage(), False=Constant(), Value=NamedRerouteUsage())
   B=StaticSwitch(True=Constant(), False=StaticSwitch(True=@n44, False=@n45, Value=@n46), Value=NamedRerouteUsage())
  )
  n48 = StaticSwitch(True=NamedRerouteUsage(), False=Constant(), Value=NamedRerouteUsage())
  n49 = NamedRerouteUsage()
  n50 = Constant()
  n51 = NamedRerouteUsage()
  n52 = MaterialFunctionCall!(Vertex Mask=Add(A=Add(A=@n47, B=@n48), B=StaticSwitch(True=@n49, False=@n50, Value=@n51))) {materialFunction=/Game/Global/Materials/Character_Parent_2023/FlipbookFace_Parent_2023/MaterialFunctions/MF_Flipbook_VertColor_WPO}
  n53 = ScalarParameter() {
   controlType=Enumeration
   defaultValue=1
   enumeration=&E_Mat_OnOff
   group=Flipbook 01 Global
   parameterName=Flipbook.ApplyWPO
   sortPriority=1
  }
  NamedRerouteDeclaration(None=LinearInterpolate(A=Constant(), B=@n52.Flipbook WPO, Alpha=@n53)) {name=Flipbook_WPO, nodeColor=<rgba 1 0 0.0938 1>}
  n54 = VectorParameter() {
   group=Flipbook 04 Mouth
   parameterName=Flipbook.MouthRemap_0-3
   sortPriority=201
   channelNames {a=3 - Mouth Press, b=2 - Mouth Neutral Wide, g=1 - Mouth Neutral Open, r=0 - Mouth Neutral Closed}
   defaultValue=<rgba 0 1 2 3>
  }
  NamedRerouteDeclaration(None=@n54.RGBA) {name=Mouth_Remap_0-3, nodeColor=<rgba 1 0.6652 0.0014 1>}
  NamedRerouteDeclaration(None=@n52.Vertex Color Hide Mask) {name=VertexColorHideMask, nodeColor=<rgba 1 0.545 0.1681 1>}
  NamedRerouteDeclaration(None=ChannelMaskParameter(None=VertexColor()) {group=Flipbook 04 Mouth, parameterName=Flipbook.MouthVertexColor}.R) {name=Mouth_VertexColor, nodeColor=<rgba 1 0.9844 0 1>}
  NamedRerouteDeclaration(None=StaticBoolParameter() {defaultValue=true, group=Flipbook 04 Mouth, parameterName=Flipbook.UseTextureForMouth, sortPriority=-20}) {name=UseTextureForMouth, nodeColor=<rgba 0 1 0.3438 1>}
  n55 = TextureObjectParameter() {
   group=Flipbook 04 Mouth
   parameterName=Flipbook.MouthTexture
   sortPriority=2
   texture=&T_black_emptyalpha
  }
  NamedRerouteDeclaration(None=@n55) {name=Mouth_FlipbookTexture, nodeColor=<rgba 1 0.1875 0 1>}
  NamedRerouteDeclaration(None=ChannelMaskParameter(None=VertexColor()) {group=Flipbook 01 Global, parameterName=Flipbook.ExtraVertexColor, sortPriority=-6}.R) {name=Extra_VertexColor, nodeColor=<rgba 1 0.9844 0 1>}
  NamedRerouteDeclaration(None=ScalarParameter() {defaultValue=3, group=Flipbook 04 Mouth, parameterName=Flipbook.MouthRowCount, sortPriority=3}) {name=Mouth_RowCount, nodeColor=<rgba 1 0 0.7734 1>}
  NamedRerouteDeclaration(None=ScalarParameter() {defaultValue=3, group=Flipbook 04 Mouth, parameterName=Flipbook.MouthColumnCount, sortPriority=4}) {name=Mouth_ColumnCount, nodeColor=<rgba 1 0 0.7734 1>}
  NamedRerouteDeclaration(None=StaticSwitchParameter(True=TextureCoordinate() {coordinateIndex=1}, False=TextureCoordinate()) {group=Flipbook 04 Mouth, parameterName=Flipbook.MouthUseUV2, sortPriority=10}) {name=Mouth_UVs, nodeColor=<rgba 1 0.5859 0 1>}
  n56 = ScalarParameter() {defaultValue=1, group=Flipbook 04 Mouth, parameterName=Flipbook.MouthUVScaleX, sortPriority=12}
  n57 = ScalarParameter() {defaultValue=1, group=Flipbook 04 Mouth, parameterName=Flipbook.MouthUVScaleY, sortPriority=13}
  NamedRerouteDeclaration(None=Multiply(
    A=ScalarParameter() {defaultValue=1, group=Flipbook 04 Mouth, parameterName=Flipbook.MouthGlobalScale, sortPriority=12}
    B=AppendVector(A=@n56, B=@n57)
   )) {name=Mouth_UVScale, nodeColor=<rgba 0.2656 0 1 1>}
  NamedRerouteDeclaration(None=AppendVector(
    A=ScalarParameter() {group=Flipbook 04 Mouth, parameterName=Flipbook.MouthUVOffsetX, sortPriority=20}
    B=ScalarParameter() {group=Flipbook 04 Mouth, parameterName=Flipbook.MouthUVOffsetY, sortPriority=21}
   )) {name=Mouth_UVOffset, nodeColor=<rgba 0.0547 1 0 1>}
  n58 = NamedRerouteUsage()
  n59 = Constant()
  n60 = NamedRerouteUsage()
  n61 = Add(
   A=StaticSwitch(True=NamedRerouteUsage(), False=Constant(), Value=NamedRerouteUsage())
   B=StaticSwitch(True=Constant(), False=StaticSwitch(True=@n58, False=@n59, Value=@n60), Value=NamedRerouteUsage())
  )
  n62 = StaticSwitch(True=NamedRerouteUsage(), False=Constant(), Value=NamedRerouteUsage())
  NamedRerouteDeclaration(None=Saturate(None=Add(A=@n61, B=@n62))) {name=Flipbook_FaceMask, nodeColor=<rgba 1 0.7969 0 1>}
  n63 = StaticSwitch(True=NamedRerouteUsage(), False=Constant(), Value=NamedRerouteUsage())
  n64 = NamedRerouteUsage()
  n65 = NamedRerouteUsage()
  n66 = StaticSwitch(
   True=@n63
   False=StaticSwitch(True=LinearInterpolate(A=@n63, B=@n64, Alpha=@n65), False=@n63, Value=NamedRerouteUsage())
   Value=NamedRerouteUsage()
  )
  n67 = NamedRerouteUsage()
  n68 = NamedRerouteUsage()
  NamedRerouteDeclaration(None=StaticSwitch(True=LinearInterpolate(A=@n66, B=@n67, Alpha=@n68), False=@n66, Value=NamedRerouteUsage())) {name=Flipbook_FaceColor, nodeColor=<rgba 1 0.849 0.0986 1>}
  NamedRerouteDeclaration(None=@n31.Emissive) {name=Emissive, nodeColor=<rgba 1 0.1875 0 1>}
  n69 = MaterialFunctionCall!(
   CustomLightVector
   OverrideNormal_WS=MaterialFunctionCall!(Normal=@n31.Normal) {materialFunction=/CosmeticCompanions/Materials/MaterialFunctions/MF_CompanionAnime_NormalSphere}.Normal
   STT=@n31.STT
   Normal_Map=@n31.STT
   DFL=@n31.STT
   FX=@n31.STT
  ) {materialFunction=/Game/Global/Materials/Character_Parent_2023/Anime_Parent_2023/MaterialFunctions/MFC_AnimeShader_2023_Calculations}
  n70 = GetMaterialAttributes(None=MaterialFunctionCall!(
    Shading_Gradient=@n69.Shading_Gradient
    Threshold_Gradient=@n69.Shading_Gradient
    Specular_Gradient=@n69.Shading_Gradient
    Sheen_Mask=@n69.Shading_Gradient
    Thickness=@n69.Shading_Gradient
    DFL_Mask=@n69.Shading_Gradient
    FX=@n69.Shading_Gradient
    Metallic_Gradient=@n69.Shading_Gradient
    RimLight_Gradient=@n69.Shading_Gradient
    NormalBuffer=@n69.Shading_Gradient
    --------
    CINE_LightSources=@n69.CINE_LightSources
    CINE_LightColor=@n69.CINE_LightColor
    ---------
    Color_Lit=NamedRerouteUsage()
    Color_Shaded=NamedRerouteUsage()
    SRM=@n31.SRM
    Emissive=NamedRerouteUsage()
    DFL_Color=@n31.DFL_Color
    Specular_Color=@n31.Specular_Color
    RimLight_Color=@n31.StyleRimLight_Color
   ) {materialFunction=/Game/Global/Materials/Character_Parent_2023/Anime_Parent_2023/MaterialFunctions/MFC_AnimeShader_2023_Post}.Result) {attributeGetTypes [#1, #2]}
  n71 = NamedRerouteUsage()
  n72 = Add(A=@n70.World Position Offset, B=NamedRerouteUsage())
  n73 = NamedRerouteUsage()
  SetMaterialAttributes!(
   MaterialAttributes=NamedRerouteDeclaration(None=@n70.MaterialAttributes) {name=Material, nodeColor=<rgba 0.6406 0 1 1>}
   Base Color=StaticSwitch(True=NamedRerouteUsage(), False=@n70.Base Color, Value=NamedRerouteUsage())
   Opacity Mask=NamedRerouteUsage()
   World Position Offset=StaticSwitch(True=NamedRerouteUsage(), False=StaticSwitch(True=@n71, False=@n72, Value=@n73), Value=NamedRerouteUsage())
  ) {attributeSetTypes [#1, #3, #2]}
  TemporalResponsivenessOutput!(None=LinearInterpolate(B=ScalarParameter() {parameterName=Temporal Responsiveness}, Alpha=NamedRerouteUsage()) {constA=0.4})
 ]
}
